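# Optimized Niagara vehicle boost effect for the LPFire cosmetic, attached to filtered skeletal-mesh sockets and split across boost ports.
# Cores merges mesh and sprite layers into one emitter using spawn groups and a VisTag.
# Separate emitters add streaming animated sprites and per-port ribbon trails that inherit source movement; an older rocks sprite emitter is disabled.
# Paint-variant overrides set colors, while painted and frontend-preview flags switch brightness, scale, offsets and motion.
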NiagaraSystem /VehicleCosmetics/Boosts/Boost_LPFire/FX/NS_Boost_LPFire_Optimized
dependencies {
 usedDataInterfaces [
  &NiagaraDataInterfaceArrayFloat3
  &NiagaraDataInterfaceCurve
  &NiagaraDataInterfaceSpriteRendererInfo
  &NiagaraDataInterfaceSkeletalMesh
 ]
 usedDynamicInputs [&VectorLength, &MakeCustomVector2DFromBool]
 usedModules [
  &self:SystemSpawnScript.SystemScriptSource.SystemScriptGraph.NiagaraNodeAssignment_0.SetVariables_#1
  &SystemState
 ]
}
emitters {
 Cores {
  emitterUpdateScript [
   &EmitterState(
    Inactive Response=::NewEnumerator1
    Life Cycle Mode=::NewEnumerator1
    Loop Behavior=::NewEnumerator1
    Loop Duration Mode=::NewEnumerator1
    UseLoopDelay=false
   )
   &VC_FilteredSocketCount()
   &SpawnRate(SpawnGroup=2, SpawnRate=&RandomRangeFloat(Minimum=25, Maximum=30), Use Spawn Probability=false)
   &SpawnBurst_Instantaneous(Spawn Count=2, Use Spawn Probability=false, UseLoopCountLimit=false)
   &SpawnBurst_Instantaneous(Spawn Count=2, Spawn Group=1, Use Spawn Probability=false, UseLoopCountLimit=false) {module=SpawnBurst_Instantaneous001}
  ]
  particleSpawnScript [
   &VC_PaintedVariantsOverride(
    Black=<rgba 0.007 0.007 0.007 1>
    BurntSienna=<rgba 0.1146 0.0191 0 1>
    Cobalt=<rgba 0 0.0206 0.4115 1>
    Crimson=<rgba 0.2031 0 0.0055 1>
    DefaultColor=<rgba 0.0225 0.185 0.1679 1>
    ForestGreen=<rgba 0.0519 0.4271 0.0178 1>
    Grey=<rgba 0.025 0.025 0.025 1>
    Lime=<rgba 0.1048 0.599 0.006 1>
    Orange=&MakeCustomLinearColorFromBool(Bool=$User.bIsFrontendPreview, True Color=<rgba 0.4635 0.0464 0 1>, False Color=<rgba 0.8 0.08 0 1>)
    Pink=<rgba 0.3438 0.0818 0.2455 1>
    Purple=<rgba 0.1076 0 0.3229 1>
    R1 Black:LinearColor=<rgba 0.003 0.003 0.003 1>
    R1 BurntSienna:LinearColor=&MakeCustomLinearColorFromBool(Bool=$User.bIsFrontendPreview, True Color=<rgba 2.5 0.4375 0.025 1>, False Color=<rgba 10 1.75 0.1 1>)
    R1 Cobalt:LinearColor=<rgba 0 0.0233 0.466 1>
    R1 Crimson:LinearColor=<rgba 0.1771 0 0.0048 1>
    R1 DefaultColor=<rgba 0 0.1833 0.2 1>
    R1 ForestGreen:LinearColor=&MakeCustomLinearColorFromBool(Bool=$User.bIsFrontendPreview, True Color=<rgba 0 0.292 0.0097 1>, False Color=<rgba 0.0467 0.4 0 1>)
    R1 Grey:LinearColor=<rgba 0.024 0.024 0.024 1>
    R1 Lime:LinearColor=<rgba 0.1758 0.2344 0 1>
    R1 Orange:LinearColor=<rgba 0.75 0.0625 0 1>
    R1 Pink:LinearColor=&MakeCustomLinearColorFromBool(Bool=$User.bIsFrontendPreview, True Color=<rgba 10 2.381 7.1428 1>, False Color=<rgba 10 2.381 7.1428 1>)
    R1 Purple:LinearColor=<rgba 0.0382 0 0.1146 1>
    R1 Saffron:LinearColor=<rgba 0.8594 0.2865 0 1>
    R1 SkyBlue:LinearColor=<rgba 0 0.0754 0.4522 1>
    R1 TitaniumWhite:LinearColor=<rgba 4 4 4 1>
    R2 Black:LinearColor=<rgba 0.006 0.006 0.006 1>
    R2 BurntSienna:LinearColor=<rgba 2.5 0.4375 0.025 1>
    R2 Cobalt:LinearColor=<rgba 0 0.035 0.7 1>
    R2 DefaultColor:LinearColor=<rgba 0 0.2 0.1833 1>
    R2 ForestGreen:LinearColor=<rgba 0 0.55 0.0183 1>
    R2 Grey:LinearColor=<rgba 0.05 0.05 0.05 1>
    R2 Lime:LinearColor=<rgba 0.3 0.4 0 1>
    R2 Orange:LinearColor=<rgba 1.2 0.1 0 1>
    R2 Pink:LinearColor=&MakeCustomLinearColorFromBool(Bool=$User.bIsFrontendPreview, True Color=<rgba 10 2.381 7.1428 1>, False Color=<rgba 2 0.4762 1.4286 1>)
    R2 Purple:LinearColor=<rgba 0.0807 0 0.22 1>
    R2 Saffron:LinearColor=<rgba 0.7 0.5367 0 1>
    R2 SkyBlue:LinearColor=<rgba 0 0.15 0.9 1>
    R2 TitaniumWhite:LinearColor=<rgba 4 4 4 1>
    Saffron=<rgba 0.1562 0.1432 0 1>
    SkyBlue=<rgba 0.0326 0.1356 0.651 1>
    TitaniumWhite=<rgba 0.1354 0.1354 0.1354 1>
    Use R1 Painted Overrides=true
    Use R2 Painted Overrides=true
   )
   /SetVariables(Particles.VisTag:NiagaraInt32=&SpawnGroup()) {isSetParameters=true}
   &InitializeParticle(
    Color=$Particles.Color
    Mesh Scale=&MakeCustomVectorFromBool(Bool=$User.bIsFrontendPreview, True Vector=<1 0.8 0.8>, False Vector=<0.8 0.8 0.8>)
    Mesh Scale Mode=::NewEnumerator4
    Position Mode=::NewEnumerator0
    Write Material Random=true
   )
   &PartitionParticles(Number Of Partitions=$Emitter.PortCount)
   &SkeletalMeshLocation(
    Bone / Socket Index=$Particles.PartitionParticles.Partition
    Bone / Socket Sampling Mode=::NewEnumerator5
    Interpolate Mesh Sampling=false
    MaskSpawnGroup=false
    Mesh Sampling Type=::NewEnumerator6
    Sampled Position Offset=&AddVectorToPosition(
     Position=&MaskVectorBySpawnGroup(Masked Vector=<0 0 0>, Passthrough Vector=<-10 0 0>)
     Vector=&MakeCustomVectorFromBool(
      Bool=$User.bIsFrontendPreview
      True Vector=$User.DTCarOffset
      False Vector=&Multiply_VectorByFloat(Vector=$User.DTCarOffset, Float=0.5)
     )
    )
    Skeletal Mesh=$System.SkeletalMesh
   )
   &AddVelocity(
    Velocity=/VehicleCosmetics/FX/Niagara/Modules/MergeEmitters/VC_VectorBySpawnGroup(
     SpawnGroup or ExecIndex:NiagaraInt32=0
     CustomInt:NiagaraInt32=0
     MaximumSpawnGroups:NiagaraInt32=3
     Vector Group 0:Vector3f=<0 0 0>
     Vector Group 1:Vector3f=<0 0 0>
     Vector Group 2:Vector3f=&RandomRangeVector(Minimum=<150 0 0>, Maximum=<300 0 0>)
     Vector Group 3:Vector3f=<0 0 0>
     Vector Group 4:Vector3f=<0 0 0>
     Vector Group 5:Vector3f=<0 0 0>
    )
    Velocity Speed Scale=/VehicleCosmetics/FX/Niagara/Modules/MergeEmitters/VC_FloatBySpawnGroup(
     SpawnGroup or ExecIndex:NiagaraInt32=0
     CustomInt:NiagaraInt32=0
     MaximumSpawnGroups:NiagaraInt32=3
     Float SG 0:NiagaraFloat=0
     Float SG 1:NiagaraFloat=0
     Float SG 2:NiagaraFloat=&MakeCustomFloatFromBool(Bool=$User.bIsFrontendPreview, True Float=-1, False Float=-0.5)
     Float SG 3:NiagaraFloat=0
     Float SG 4:NiagaraFloat=0
     Float SG 5:NiagaraFloat=0
    )
   ) {enabled=false}
   &InitialMeshOrientation(Mesh Orientation Mode=::NewEnumerator2, Orientation Vector=<-1 0 0>, Use Rotation Vector=false)
   &CameraOffset(Camera Offset Amount=&VC_FloatBySpawnGroup(
     SpawnGroup or ExecIndex:NiagaraInt32=0
     CustomInt:NiagaraInt32=0
     MaximumSpawnGroups:NiagaraInt32=2
     Float SG 0:NiagaraFloat=0
     Float SG 1:NiagaraFloat=60
     Float SG 2:NiagaraFloat=100
     Float SG 3:NiagaraFloat=0
     Float SG 4:NiagaraFloat=0
     Float SG 5:NiagaraFloat=0
    ))
  ]
  particleUpdateScript [
   &ParticleState(Kill Particles When Lifetime Has Elapsed=false, Loop Particles Lifetime=true)
   &SolveForcesAndVelocity(Clamp Velocity=false, Limit Acceleration=false)
   &ScaleMeshSize(Scale Factor=&MaskVectorBySpawnGroup(
     Masked Vector=<1 1 1>
     Passthrough Vector=&MakeCustomVectorFromBool(
      Bool=$User.bIsFrontendPreview
      True Vector=&VectorFromFloat(Value=0.28)
      False Vector=&VectorFromFloat(Value=0.3)
     )
    ))
   &ScaleColor(
    Scale RGB=&MakeCustomVectorFromBool(
     Bool=$User.bIsPainted
     True Vector=&Multiply_VectorByFloat(
      Vector=<1 1 1>
      Float=&MakeCustomFloatFromBool(Bool=$User.bIsFrontendPreview, True Float=1.7, False Float=0.5)
     )
     False Vector=&Multiply_VectorByFloat(
      Vector=&VectorFromFloat(Value=0.7)
      Float=&MakeCustomFloatFromBool(Bool=$User.bIsFrontendPreview, True Float=1.8, False Float=0.75)
     )
    )
    ScaleA=true
    ScaleRGB=true
   ) {module=ScaleColor001}
   /SetVariables(Particles.R1 Color:LinearColor=&VC_DynamicScaleColor(
     Color to Scale:LinearColor=$Particles.Initial.R1 Color
     RGB Scale:Vector3f=&Multiply_VectorByFloat(
      Vector=<1 1 1>
      Float=&MakeCustomFloatFromBool(
       Bool=$User.bIsFrontendPreview
       True Float=&MakeCustomFloatFromBool(Bool=$User.bIsPainted, True Float=10, False Float=8)
       False Float=&MakeCustomFloatFromBool(Bool=$User.bIsPainted, True Float=2.5, False Float=7)
      )
     )
     Alpha Scale:NiagaraFloat=1
    )) {isSetParameters=true}
  ]
  rendererClasses [&NiagaraMeshRendererProperties, &NiagaraSpriteRendererProperties]
  renderers [
   {rendererClass=&NiagaraMeshRendererProperties, rendererIndex=0}
   {rendererClass=&NiagaraMeshRendererProperties, rendererIndex=1}
   {rendererClass=&NiagaraSpriteRendererProperties, rendererIndex=2}
  ]
 }
 NE_VC_Sprite_InitialRocks {
  bEnabled=false
  emitterSpawnScript [/SetVariables(Emitter.IsInGame:NiagaraBool=&Boolean_NotOperation(Bool=$User.bIsFrontendPreview)) {isSetParameters=true}]
  emitterUpdateScript [
   &EmitterState(Inactive Response=::NewEnumerator1, Life Cycle Mode=::NewEnumerator1, Loop Duration=0.1, UseLoopDelay=false)
   &SpawnBurst_Instantaneous(
    Spawn Count=&MakeCustomIntFromBool(Bool=$User.bIsFrontendPreview, True INT=&RandomRangeInt(Minimum=25, Maximum=35), False INT=25)
    Use Spawn Probability=false
    UseLoopCountLimit=false
   ) {enabled=false}
   &VC_FilteredSocketCount()
   &SpawnRate(SpawnRate=&RandomRangeFloat(Minimum=25, Maximum=35), Use Spawn Probability=true)
  ]
  particleSpawnScript [
   &VC_PaintedVariantsOverride(
    Black=<rgba 0.006 0.006 0.006 1>
    BurntSienna=<rgba 2.5 0.4375 0.025 1>
    Cobalt=<rgba 0 0.035 0.7 1>
    DefaultColor=<rgba 0 0.2 0.1833 1>
    ForestGreen=<rgba 0 0.55 0.0183 1>
    Grey=<rgba 0.05 0.05 0.05 1>
    Lime=<rgba 0.3 0.4 0 1>
    Orange=<rgba 1.2 0.1 0 1>
    Pink=&MakeCustomLinearColorFromBool(Bool=$User.bIsFrontendPreview, True Color=<rgba 10 2.381 7.1428 1>, False Color=<rgba 2 0.4762 1.4286 1>)
    Purple=<rgba 0.0807 0 0.22 1>
    Saffron=<rgba 0.7 0.5367 0 1>
    SkyBlue=<rgba 0 0.15 0.9 1>
    TitaniumWhite=<rgba 4 4 4 1>
   )
   &InitializeParticle(
    Color=$Particles.Color
    Lifetime Max=0.2
    Lifetime Min=0.1
    Lifetime Mode=::NewEnumerator1
    Sprite Size=&RandomRangeVector2D(Minimum=<30 50>, Maximum=<20 30>)
    Sprite Size Mode=::NewEnumerator4
    UsePositionOffset=false
    Write Material Random=true
   )
   &PartitionParticles(Number Of Partitions=$Emitter.PortCount)
   &SkeletalMeshLocation(
    Bone / Socket Index=$Particles.PartitionParticles.Partition
    Bone / Socket Sampling Mode=::NewEnumerator5
    MaskSpawnGroup=false
    Mesh Sampling Type=::NewEnumerator6
    Sampled Position Offset=&AddVectorToPosition(Position=<-45 0 -2.5>, Vector=$User.DTCarOffset)
    Sampled Position Offset Coordinate Space:ENiagaraSkelMeshTransforms=::NewEnumerator2
    Skeletal Mesh=$System.SkeletalMesh
   )
   &ApplyInitialForces()
   &AlignSpriteToMeshOrientation(
    Mesh Orientation Relative Sprite Facing Vector:Vector3f=&SelectVectorFromArray(
     Array Sampling Mode=::NewEnumerator0
     Direct Array Index=?
     Vector Selection Array=DataInterface() {propertyValues {FloatData [<1 1 0>, <0 1 0>]}}
     Random Seed=0
    )
    SetAlignment=false
    SetFacing=true
   ) {module=AlignSpriteToMeshOrientation001}
   &AddVelocity(Velocity=&RandomRangeVector(Minimum=<150 0 0>, Maximum=<300 0 0>), Velocity Speed Scale=-0.5)
   &CameraOffset(Camera Offset Amount=100) {enabled=false}
  ]
  particleUpdateScript [
   &ParticleState()
   &ScaleSpriteSize(
    Uniform Curve Scale=0.5
    Uniform Curve Sprite Scale=DataInterface() {propertyValues {Curve {keys [
        {tangentMode=RCTM_User, time=0, value=1}
        {arriveTangent=-0.0405, interpMode=RCIM_Cubic, leaveTangent=-0.0405, tangentMode=RCTM_User, time=0.7009, value=1}
        {arriveTangent=-8.0089, interpMode=RCIM_Cubic, leaveTangent=-8.0089, tangentMode=RCTM_User, time=1, value=0}
       ]}}}
   )
   &ScaleColor(
    Scale RGB=&Multiply_VectorByFloat(Vector=<1 1 1>, Float=&MakeCustomFloatFromBool(Bool=$User.bIsPainted, True Float=0.5, False Float=5))
    ScaleA=true
    ScaleRGB=true
   ) {module=ScaleColor001}
   &SolveForcesAndVelocity(Clamp Velocity=false, Limit Acceleration=false)
   &SubUVAnimation(
    "Sprite Renderer "=DataInterface() {propertyValues {SpriteRenderer=&self:NiagaraEmitter_0.NiagaraSpriteRendererProperties_1}}
    SubUV Animation Mode:ENiagara_SubUVLookupModeV2=::NewEnumerator5
   )
  ]
  rendererClasses [&NiagaraSpriteRendererProperties]
  renderers [{rendererClass=&NiagaraSpriteRendererProperties}]
 }
 NE_VC_Sprite_InitialRocks001 {
  emitterUpdateScript [
   &EmitterState(Life Cycle Mode=::NewEnumerator1, Loop Duration=1, UseLoopDelay=false)
   &VC_FilteredSocketCount()
   &SpawnRate(
    SpawnRate=&MakeCustomFloatFromBool(Bool=$User.bIsFrontendPreview, True Float=&RandomRangeFloat(Minimum=20, Maximum=40), False Float=100)
    Use Spawn Probability=false
   )
  ]
  particleSpawnScript [
   &VC_PaintedVariantsOverride(
    Black=<rgba 0.02 0.02 0.02 1>
    BurntSienna=<rgba 0.14 0.0245 0.0014 1>
    Cobalt=<rgba 0.0089 0.0676 0.8896 1>
    Crimson=<rgba 0.5 0 0.0136 1>
    DefaultColor=<rgba 0 0.4 0.3667 1>
    ForestGreen=<rgba 0.1225 0.7 0.07 1>
    Grey=<rgba 0.1 0.1 0.1 1>
    Lime=<rgba 0.25 1 0 1>
    Orange=<rgba 1 0.1333 0 1>
    Pink=<rgba 1 0.2381 0.7143 1>
    Purple=<rgba 0.3 0 0.9 1>
    R1 Black:LinearColor=<rgba 0.03 0.03 0.03 1>
    R1 BurntSienna:LinearColor=<rgba 1 0.3088 0.1233 1>
    R1 Cobalt:LinearColor=<rgba 0 0.0169 0.913 1>
    R1 Crimson:LinearColor=<rgba 0.8177 0 0.0222 1>
    R1 DefaultColor=<rgba 0.0771 0.4 0.3731 1>
    R1 ForestGreen:LinearColor=<rgba 0.238 0.8958 0.2712 1>
    R1 Lime:LinearColor=<rgba 0.4932 0.9375 0.0488 1>
    R1 Orange:LinearColor=<rgba 0.9115 0.3179 0 1>
    R1 Pink:LinearColor=<rgba 5 0.599 3.3996 1>
    R1 Saffron:LinearColor=<rgba 5 3.9488 0.4948 1>
    Saffron=<rgba 0.45 0.375 0 1>
    SkyBlue=<rgba 0.1 0.34 1 1>
    TitaniumWhite=<rgba 0.3 0.3 0.3 1>
    Use R1 Painted Overrides=true
   )
   &InitializeParticle(
    Color=$Particles.Color
    Lifetime Max=0.6
    Lifetime Min=0.3
    Lifetime Mode=::NewEnumerator1
    Sprite Rotation Mode=::NewEnumerator2
    Sprite Size Mode=::NewEnumerator1
    Uniform Sprite Size Max=20
    Uniform Sprite Size Min=12
    UsePositionOffset=false
    Write Material Random=true
   )
   &PartitionParticles(Number Of Partitions=$Emitter.PortCount)
   &SkeletalMeshLocation(
    Bone / Socket Index=$Particles.PartitionParticles.Partition
    Bone / Socket Sampling Mode=::NewEnumerator5
    MaskSpawnGroup=false
    Mesh Sampling Type=::NewEnumerator6
    Sampled Position Offset=&AddVectorToPosition(
     Position=&MakeCustomVectorFromBool(Bool=$User.bIsFrontendPreview, True Vector=<-50 0 0>, False Vector=<0 0 0>)
     Vector=$User.DTCarOffset
    )
    Skeletal Mesh=$System.SkeletalMesh
   )
   &ApplyInitialForces()
   &AddVelocity(
    Velocity=&RandomRangeVector(Minimum=<400 0 0>, Maximum=<750 0 0>)
    Velocity Speed Scale=&MakeCustomFloatFromBool(Bool=$User.bIsFrontendPreview, True Float=-1, False Float=-2.5)
   )
   &CameraOffset(Camera Offset Amount=101)
   &ShapeLocation(
    MaskSpawnGroup=false
    Override Seed=false
    Sphere Radius=&RandomRangeFloat(Minimum=20, Maximum=25)
    UseOwnerScale=true
   )
  ]
  particleUpdateScript [
   &ParticleState()
   &InheritSourceMovement(
    Applied Position Scale=&VectorFromFloat(Value=&FloatFromCurve(FloatCurve=DataInterface() {propertyValues {Curve {keys [<at 0.5 1 RCIM_Cubic>, {arriveTangent=-1.2348, interpMode=RCIM_Cubic, leaveTangent=-1.2348, tangentMode=RCTM_User, time=1, value=0.8}]}}}))
    ApplyForce=false
    ApplyPosition=true
    ApplyPositionFalloff=false
   )
   &SpriteRotationRate(Rotation Rate=100)
   &ScaleSpriteSize(
    Uniform Curve Scale=&MakeCustomFloatFromBool(Bool=$User.bIsFrontendPreview, True Float=1, False Float=0.7)
    Uniform Curve Sprite Scale=DataInterface() {propertyValues {Curve {keys [
        {tangentMode=RCTM_User, time=0, value=1}
        {arriveTangent=-0.0405, interpMode=RCIM_Cubic, leaveTangent=-0.0405, tangentMode=RCTM_User, time=0.5, value=1}
        {arriveTangent=-3.7185, interpMode=RCIM_Cubic, leaveTangent=-3.7185, tangentMode=RCTM_User, time=1, value=0}
       ]}}}
   )
   &SolveForcesAndVelocity(Clamp Velocity=false, Limit Acceleration=false)
   &SubUVAnimation(
    Random Start Frame=true
    "Sprite Renderer "=DataInterface() {propertyValues {SpriteRenderer=&self:NiagaraEmitter_2.NiagaraSpriteRendererProperties_0}}
    UseEndFrame=false
    UseStartFrame=false
   )
   &DynamicMaterialParameters(
    Index 0 Param 1=&MakeFloatFromInt(Int=$User.CoreBrightness_int)
    Module.Index0Param0Write=true
    Module.Index0Param1Write=true
    Module.Index0Param2Write=true
    Module.Index0Param3Write=true
   ) {enabled=false}
   /SetVariables(Particles.R2Color:LinearColor=&MultiplyLinearColor(
     Linear Color=$Particles.R1 Color
     Scale Factor=&MakeVector4(
      X=5
      Y=5
      Z=5
      W=&MakeCustomFloatFromBool(
       Bool=$User.bIsFrontendPreview
       True Float=&FloatFromCurve(FloatCurve=DataInterface() {propertyValues {Curve {keys [
            {arriveTangent=0.0266, interpMode=RCIM_Cubic, leaveTangent=0.0266, tangentMode=RCTM_User, time=0, value=0}
            {arriveTangent=1.1179, interpMode=RCIM_Cubic, leaveTangent=1.1179, tangentMode=RCTM_User, time=0.25, value=0.6228}
            {arriveTangent=0.0224, interpMode=RCIM_Cubic, leaveTangent=0.0224, tangentMode=RCTM_User, time=1, value=1}
           ]}}})
       False Float=&FloatFromCurve(FloatCurve=DataInterface() {propertyValues {Curve {keys [
            {arriveTangent=9.0725, interpMode=RCIM_Cubic, leaveTangent=9.0725, tangentMode=RCTM_User, time=0, value=0}
            {arriveTangent=-2.5249, interpMode=RCIM_Cubic, leaveTangent=-2.5249, tangentMode=RCTM_User, time=0.1408, value=0.9847}
            {arriveTangent=-0.0405, interpMode=RCIM_Cubic, leaveTangent=-0.0405, tangentMode=RCTM_User, time=1, value=0}
           ]}}})
      )
     )
    )) {isSetParameters=true}
   &ScaleColor(
    Scale Alpha=&MakeCustomFloatFromBool(
     Bool=$User.bIsFrontendPreview
     True Float=&FloatFromCurve(FloatCurve=DataInterface() {propertyValues {Curve {keys [
          {arriveTangent=0.0266, interpMode=RCIM_Cubic, leaveTangent=0.0266, tangentMode=RCTM_User, time=0, value=0}
          {arriveTangent=1.1179, interpMode=RCIM_Cubic, leaveTangent=1.1179, tangentMode=RCTM_User, time=0.25, value=0.6228}
          {arriveTangent=0.0224, interpMode=RCIM_Cubic, leaveTangent=0.0224, tangentMode=RCTM_User, time=1, value=1}
         ]}}})
     False Float=&FloatFromCurve(FloatCurve=DataInterface() {propertyValues {Curve {keys [
          {arriveTangent=9.0725, interpMode=RCIM_Cubic, leaveTangent=9.0725, tangentMode=RCTM_User, time=0, value=0}
          {arriveTangent=-2.5249, interpMode=RCIM_Cubic, leaveTangent=-2.5249, tangentMode=RCTM_User, time=0.1408, value=0.9847}
          {arriveTangent=-0.0405, interpMode=RCIM_Cubic, leaveTangent=-0.0405, tangentMode=RCTM_User, time=1, value=0}
         ]}}})
    )
    Scale RGB=&Multiply_VectorByFloat(
     Vector=<1 1 1>
     Float=&MakeCustomFloatFromBool(
      Bool=$User.bIsFrontendPreview
      True Float=0.5
      False Float=&MakeCustomFloatFromBool(Bool=$User.bIsPainted, True Float=0.3, False Float=0.7)
     )
    )
    ScaleA=true
    ScaleRGB=true
   ) {module=ScaleColor001}
  ]
  rendererClasses [&NiagaraSpriteRendererProperties]
  renderers [
   {rendererClass=&NiagaraSpriteRendererProperties}
   {rendererClass=&NiagaraSpriteRendererProperties, rendererIndex=1}
  ]
 }
 Ribbon {
  emitterUpdateScript [
   &EmitterState(
    Inactive Response=::NewEnumerator1
    Life Cycle Mode=::NewEnumerator1
    Loop Duration=&MakeCustomFloatFromBool(Bool=$User.bIsFrontendPreview, True Float=1, False Float=0.1)
    Recalculate Duration Each Loop=true
    UseLoopDelay=false
   )
   &VC_FilteredSocketCount()
   &VC_RateQualityHandler()
   &VC_SpawnRibbon_v1(Number of Ribbons=$Emitter.PortCount, SpawnRate=25)
  ]
  particleSpawnScript [
   &VC_PaintedVariantsOverride(
    Black=<rgba 0.005 0.005 0.005 1>
    BurntSienna=<rgba 0.07 0.0117 0 1>
    Cobalt=<rgba 0 0.0117 0.234 1>
    Crimson=<rgba 0.1354 0 0.0037 1>
    DefaultColor=<rgba 0.011 0.0938 0.087 2>
    ForestGreen=<rgba 0.0181 0.0938 0.0112 1>
    Grey=<rgba 0.015 0.015 0.015 1>
    Lime=<rgba 0.0556 0.3333 0 1>
    Orange=<rgba 0.3333 0.0222 0 1>
    Pink=&MakeCustomLinearColorFromBool(Bool=$User.bIsFrontendPreview, True Color=<rgba 0.1927 0.0459 0.1376 1>, False Color=<rgba 0.7 0.1667 0.5 1>)
    Purple=<rgba 0.0625 0 0.1875 1>
    R1 Black:LinearColor=<rgba 0.002 0.002 0.002 1>
    R1 BurntSienna:LinearColor=<rgba 1 0.1867 0.024 1>
    R1 Cobalt:LinearColor=<rgba 0 0.0527 0.2552 1>
    R1 Crimson:LinearColor=<rgba 0.151 0 0.0041 1>
    R1 DefaultColor=<rgba 0 0.111 0.111 0.15>
    R1 ForestGreen:LinearColor=<rgba 0 0.169 0.0085 1>
    R1 Grey:LinearColor=<rgba 0.014 0.014 0.014 0.2>
    R1 Lime:LinearColor=<rgba 0.0911 0.1823 0 1>
    R1 Orange:LinearColor=<rgba 0.2865 0.0999 0 1>
    R1 Pink:LinearColor=<rgba 5 0.599 3.3996 1>
    R1 Purple:LinearColor=<rgba 0.0321 0 0.0781 1>
    R1 Saffron:LinearColor=<rgba 0.268 0.2055 0 1>
    R1 SkyBlue:LinearColor=<rgba 0 0.0865 0.13 0.3>
    Saffron=<rgba 0.0781 0.0651 0 1>
    SkyBlue=<rgba 0.0203 0.0846 0.4062 1>
    TitaniumWhite=<rgba 0.0625 0.0625 0.0625 1>
    Use R1 Painted Overrides=true
   )
   &VC_InitializeRibbon_v1(
    Color=$Particles.Color
    Custom Offset=true
    Enable Skeletal Position=true
    Lifetime=&MakeCustomFloatFromBool(Bool=$User.bIsFrontendPreview, True Float=0.8, False Float=0.6)
    Local Vector Offset=&AddVectorToPosition(
     Position=&MakeCustomVectorFromBool(Bool=$User.bIsFrontendPreview, True Vector=<-60 0 -1>, False Vector=<-60 0 -10>)
     Vector=$User.DTCarOffset
    )
    Offset Amount=1
    Ribbon Width=40
   )
   &AddVelocity(
    Velocity=&MakeCustomVectorFromBool(Bool=$User.bIsFrontendPreview, True Vector=<1000 0 0>, False Vector=<1500 0 0>)
    Velocity Speed Scale=-1
   ) {module=AddVelocity001}
   &CameraOffset(Camera Offset Amount=&MakeCustomFloatFromBool(Bool=$User.bIsFrontendPreview, True Float=100, False Float=100))
  ]
  particleUpdateScript [
   &ParticleState()
   &InheritSourceMovement(
    Applied Position Scale=&Multiply_VectorByFloat(
     Vector=<1 1 1>
     Float=&FloatFromCurve(FloatCurve=DataInterface() {propertyValues {Curve {keys [<at 0.3 1 RCIM_Cubic>, {arriveTangent=-0.0312, interpMode=RCIM_Cubic, leaveTangent=-0.0312, tangentMode=RCTM_User, time=1, value=0.5}]}}})
    )
    ApplyForce=false
    ApplyPosition=true
    ApplyPositionFalloff=false
   )
   &SolveForcesAndVelocity(Clamp Velocity=false, Limit Acceleration=false)
   &VC_UpdateRibbon_v1(Enable Skeletal Position=true)
   &ScaleColor(
    Scale Alpha=&Multiply_Float(
     A=&MakeCustomFloatFromBool(Bool=$User.bIsPainted, True Float=2, False Float=1)
     B=&FloatFromCurve(FloatCurve=DataInterface() {propertyValues {Curve {keys [<at 0 1>,<at 0.5 1>,<at 1 0>]}}})
    )
    Scale RGB=&Multiply_VectorByFloat(
     Vector=&VectorFromFloat(Value=1.9)
     Float=&MakeCustomFloatFromBool(Bool=$User.bIsFrontendPreview, True Float=3, False Float=0.9)
    )
    ScaleA=true
    ScaleRGB=true
   )
   &ScaleRibbonWidth(Ribbon Width Scale=&FloatFromCurve(FloatCurve=DataInterface() {propertyValues {Curve {keys [
         {tangentMode=RCTM_User, time=0, value=1}
         {arriveTangent=-0.0405, leaveTangent=-0.0405, time=0.5, value=1}
         {arriveTangent=-8.0089, leaveTangent=-8.0089, time=1, value=0}
        ]}}}))
   /SetVariables(Particles.Ribbon2Color:LinearColor=&MultiplyLinearColor(
     Linear Color=$Particles.R1 Color
     Scale Factor=/Niagara/DynamicInputs/Multiply/Multiply_Vector4_ByFloat(
      Vector4:Vector4f=<1 1 1 1>
      Float:NiagaraFloat=&MakeCustomFloatFromBool(Bool=$User.bIsFrontendPreview, True Float=5, False Float=1)
     )
    )) {isSetParameters=true}
  ]
  rendererClasses [&NiagaraRibbonRendererProperties]
  renderers [
   {rendererClass=&NiagaraRibbonRendererProperties}
   {rendererClass=&NiagaraRibbonRendererProperties, rendererIndex=1}
  ]
 }
}
summary {
 systemName=NS_Boost_LPFire_Optimized
 emitters [
  {bEnabled=false, emitterName=NE_VC_Sprite_InitialRocks, rendererClasses [&NiagaraSpriteRendererProperties]}
  {emitterName=Ribbon, rendererClasses [&NiagaraRibbonRendererProperties]}
  {emitterName=NE_VC_Sprite_InitialRocks001, rendererClasses [&NiagaraSpriteRendererProperties]}
  {emitterName=Cores, rendererClasses [&NiagaraMeshRendererProperties, &NiagaraSpriteRendererProperties]}
 ]
 userVariables [
  User.bIsFrontendPreview {type=NiagaraBool}
  User.bIsPainted {type=NiagaraBool}
  User.ColorVariantID {type=NiagaraInt32}
  User.CoreBrightness_int {type=NiagaraInt32, defaultValue=5}
  User.DTCarOffset {type=Vector3f, description="", defaultValue {struct=Vector3f, value=<0 0 0>}}
  User.PaintedColor {type=LinearColor, description="", defaultValue {struct=LinearColor, value=<rgba 1 1 1 1>}}
  User.PaintedScalar {type=NiagaraFloat, defaultValue=5, description=""}
  User.PortCount {type=NiagaraInt32, defaultValue=2}
 ]
}
